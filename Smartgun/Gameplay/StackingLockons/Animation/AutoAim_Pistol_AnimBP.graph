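# Animation Blueprint for the stacking-lockons auto-aim pistol, animating it between non-targeted and targeted poses.
# An entry conduit picks the starting pose from bIsTargeting, which also triggers one-shot transition clips that advance automatically on completion.
# Retargeting mid-way through the return clip cuts to NonTargetedPose, while the outgoing clip has no early exit.
# All crossfades are zero, DefaultSlot follows the state machine, and the update event is empty.
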AnimBlueprint /Smartgun/Gameplay/StackingLockons/Animation/AutoAim_Pistol_AnimBP
target_skeleton=/Smartgun/Gameplay/StackingLockons/Meshes/AutoAim_Pistol_Skeleton
graphs [
 AnimGraph {nodes [
   AnimGraphNode_StateMachine_0 = AimTransition() {is=title AnimGraphNode_StateMachine, bAllowConduitEntryStates=true}
   Output Pose(Result=Slot 'DefaultSlot'(Source=@AnimGraphNode_StateMachine_0)) {layerGroup=None}
  ]}
 AimTransition {
  entry=Conduit
  path=AnimGraph.AnimGraphNode_StateMachine_0.AimTransition
  ids=id
  nodes [
   AnimStateConduitNode_0 = Conduit()
   AnimStateNode_0 = NonTargetedPose() {is=title AnimStateNode}
   AnimStateNode_1 = TransitionToTargeted() {is=title AnimStateNode}
   AnimStateNode_2 = TargetedPose() {is=title AnimStateNode}
   AnimStateNode_3 = TransitionToNonTargeted() {is=title AnimStateNode}
   AnimStateTransitionNode_0 = AnimStateTransitionNode() {crossfadeDuration=0}
   AnimStateTransitionNode_1 = AnimStateTransitionNode() {bAutomaticRuleBasedOnSequencePlayerInState=true, crossfadeDuration=0}
   AnimStateTransitionNode_2 = AnimStateTransitionNode() {crossfadeDuration=0}
   AnimStateTransitionNode_3 = AnimStateTransitionNode() {bAutomaticRuleBasedOnSequencePlayerInState=true, crossfadeDuration=0, priorityOrder=2}
   AnimStateTransitionNode_4 = AnimStateTransitionNode() {crossfadeDuration=0, priorityOrder=100}
   AnimStateTransitionNode_5 = AnimStateTransitionNode() {crossfadeDuration=0}
   AnimStateTransitionNode_6 = AnimStateTransitionNode() {crossfadeDuration=0}
  ]
  links [
   {
    dst=NonTargetedPose
    rule=AnimGraph.AnimGraphNode_StateMachine_0.AimTransition.AnimStateTransitionNode_4.Transition
    src=Conduit
    via=AnimStateTransitionNode_4
   }
   {
    dst=TargetedPose
    rule=AnimGraph.AnimGraphNode_StateMachine_0.AimTransition.AnimStateTransitionNode_6.Transition
    src=Conduit
    via=AnimStateTransitionNode_6
   }
   {
    dst=TransitionToTargeted
    rule=AnimGraph.AnimGraphNode_StateMachine_0.AimTransition.AnimStateTransitionNode_0.Transition
    src=NonTargetedPose
    via=AnimStateTransitionNode_0
   }
   {
    dst=TransitionToNonTargeted
    rule=AnimGraph.AnimGraphNode_StateMachine_0.AimTransition.AnimStateTransitionNode_2.Transition
    src=TargetedPose
    via=AnimStateTransitionNode_2
   }
   {
    dst=NonTargetedPose
    rule=AnimGraph.AnimGraphNode_StateMachine_0.AimTransition.AnimStateTransitionNode_3.Transition
    src=TransitionToNonTargeted
    via=AnimStateTransitionNode_3
   }
   {
    dst=NonTargetedPose
    rule=AnimGraph.AnimGraphNode_StateMachine_0.AimTransition.AnimStateTransitionNode_5.Transition
    src=TransitionToNonTargeted
    via=AnimStateTransitionNode_5
   }
   {
    dst=TargetedPose
    rule=AnimGraph.AnimGraphNode_StateMachine_0.AimTransition.AnimStateTransitionNode_1.Transition
    src=TransitionToTargeted
    via=AnimStateTransitionNode_1
   }
  ]
 }
 Conduit {path=AnimGraph.AnimGraphNode_StateMachine_0.AimTransition.AnimStateConduitNode_0.Conduit, nodes [Result(bCanEnterTransition=true)]}
 NonTargetedPose {
  path=AnimGraph.AnimGraphNode_StateMachine_0.AimTransition.AnimStateNode_0.NonTargetedPose
  nodes [Output Animation Pose(Result=Sequence Player 'Pistol_AutoAim_NonTargeted_Pose_WMM'() {sequence=&/Pistol_AutoAim_NonTargeted_Pose_WMM})]
 }
 TransitionToTargeted {
  path=AnimGraph.AnimGraphNode_StateMachine_0.AimTransition.AnimStateNode_1.TransitionToTargeted
  nodes [Output Animation Pose(Result=Sequence Player 'Pistol_AutoAim_NonTargetedToTargeted_CMM_WMM'() {bLoopAnimation=false, sequence=&/Pistol_AutoAim_NonTargetedToTargeted_CMM_WMM})]
 }
 TargetedPose {
  path=AnimGraph.AnimGraphNode_StateMachine_0.AimTransition.AnimStateNode_2.TargetedPose
  nodes [Output Animation Pose(Result=Sequence Player 'Pistol_AutoAim_Targeted_Pose_WMM'() {sequence=&/Pistol_AutoAim_Targeted_Pose_WMM})]
 }
 TransitionToNonTargeted {
  path=AnimGraph.AnimGraphNode_StateMachine_0.AimTransition.AnimStateNode_3.TransitionToNonTargeted
  nodes [Output Animation Pose(Result=Sequence Player 'Pistol_AutoAim_TargetedToNonTargeted_CMM_WMM'() {bLoopAnimation=false, sequence=&/Pistol_AutoAim_TargetedToNonTargeted_CMM_WMM})]
 }
 Transition {
  path=AnimGraph.AnimGraphNode_StateMachine_0.AimTransition.AnimStateTransitionNode_0.Transition
  nodes [Result(bCanEnterTransition=Get bIsTargeting())]
 }
 Transition {path=AnimGraph.AnimGraphNode_StateMachine_0.AimTransition.AnimStateTransitionNode_1.Transition, nodes [Result()]}
 Transition {
  path=AnimGraph.AnimGraphNode_StateMachine_0.AimTransition.AnimStateTransitionNode_2.Transition
  nodes [Result(bCanEnterTransition=NOT Boolean(A=Get bIsTargeting()))]
 }
 Transition {path=AnimGraph.AnimGraphNode_StateMachine_0.AimTransition.AnimStateTransitionNode_3.Transition, nodes [Result()]}
 Transition {path=AnimGraph.AnimGraphNode_StateMachine_0.AimTransition.AnimStateTransitionNode_4.Transition, nodes [Result(bCanEnterTransition=true)]}
 Transition {
  path=AnimGraph.AnimGraphNode_StateMachine_0.AimTransition.AnimStateTransitionNode_5.Transition
  nodes [Result(bCanEnterTransition=Get bIsTargeting())]
 }
 Transition {
  path=AnimGraph.AnimGraphNode_StateMachine_0.AimTransition.AnimStateTransitionNode_6.Transition
  nodes [Result(bCanEnterTransition=Get bIsTargeting())]
 }
 EventGraph {nodes [Event BlueprintUpdateAnimation()]}
]
variables []
