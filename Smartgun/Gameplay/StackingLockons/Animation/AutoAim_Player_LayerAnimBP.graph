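# Player layer Anim Blueprint for the Smartgun stacking lock-on feature, playing an auto-aim "targeted" reaction on the body.
# Only the MovementMode SourcePose layer has logic; bPlayUpperBodyTargeting switches between the cached input pose and crouch or standing state machines.
# Each machine applies a one-shot pistol reaction additively over that pose, automatically returns to it, and fires Targeted_Out on leaving.
# The other MovementMode layers and main AnimGraph are unwired stubs.
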AnimBlueprint /Smartgun/Gameplay/StackingLockons/Animation/AutoAim_Player_LayerAnimBP
target_skeleton=&Fortnite_M_Avg_Player_Skeleton
graphs [
 AnimGraph {nodes [Output Pose() {layerGroup=None}]}
 EventGraph {nodes [
   TryGetPawnOwner()
   Event BlueprintUpdateAnimation()
  ]}
 MovementMode_AimOffset {nodes [
   Input Pose() {name=InAimOffsetsPose}
   Input Pose() {name=InSourcePosePreAimOffsets}
   Output Pose() {layerGroup=MovementMode}
  ]}
 MovementMode_BaseLocomotion {nodes [
   Input Pose() {name=InPoseBaseLocomotion}
   Output Pose() {layerGroup=MovementMode}
  ]}
 MovementMode_FinalPoseOverride {nodes [
   Input Pose() {name=InFinalPose}
   Output Pose() {layerGroup=MovementMode}
  ]}
 MovementMode_FullBodyOverride {nodes [
   Input Pose() {name=InputPoseFullBody}
   Output Pose() {layerGroup=MovementMode}
  ]}
 MovementMode_InAirOverride {nodes [
   Input Pose() {name=InSource_Pose}
   Input Pose() {name=InAirPose}
   Output Pose() {layerGroup=MovementMode}
  ]}
 MovementMode_LocomotionAdditives {nodes [
   Input Pose() {name=InLocomotionAdditivesPose}
   Output Pose() {layerGroup=MovementMode}
  ]}
 MovementMode_LowerBodyOverride {nodes [
   Input Pose() {name=InLowerBodyPose}
   Input Pose() {name=InSourcePose}
   Output Pose() {layerGroup=MovementMode}
  ]}
 MovementMode_OffScreen {nodes [
   Input Pose() {name=InPoseOffscreen}
   Output Pose() {layerGroup=MovementMode}
  ]}
 MovementMode_SourcePose {nodes [
   AnimGraphNode_StateMachine_0 = CrouchTargetedState() {is=title AnimGraphNode_StateMachine, bAllowConduitEntryStates=true}
   AnimGraphNode_StateMachine_4 = TargetedState() {is=title AnimGraphNode_StateMachine, bAllowConduitEntryStates=true}
   Output Pose(Result=Blend Poses by bool(
     BlendPose_0=Blend Poses by bool(BlendPose_0=@AnimGraphNode_StateMachine_0, BlendPose_1=@AnimGraphNode_StateMachine_4) {blendType=Linear, childUpateMode=ResetChildOnActivate}
     BlendPose_1=Use cached pose 'InSourcePose'()
     bActiveValue=Get bPlayUpperBodyTargeting() -> (bPlayUpperBodyTargeting:bool)
    ) {blendType=Linear, childUpateMode=ResetChildOnActivate}) {layerGroup=MovementMode}
   Save cached pose 'InSourcePose'(Pose=Input Pose() {name=InSourcePoseBase}) {cacheName=InSourcePose}
  ]}
 CrouchTargetedState {
  entry=CrouchTargetedReaction
  path=MovementMode_SourcePose.AnimGraphNode_StateMachine_0.CrouchTargetedState
  ids=id
  nodes [
   AnimStateNode_1 = CrouchTargetedReaction() {is=title AnimStateNode, bAlwaysResetOnEntry=true, stateLeft {guid=#1, notifyName=Targeted_Out}}
   AnimStateNode_3 = CrouchTargetedSourcePose() {is=title AnimStateNode}
   AnimStateTransitionNode_3 = AnimStateTransitionNode() {bAutomaticRuleBasedOnSequencePlayerInState=true}
  ]
  links [{
    dst=CrouchTargetedSourcePose
    rule=MovementMode_SourcePose.AnimGraphNode_StateMachine_0.CrouchTargetedState.AnimStateTransitionNode_3.Transition
    src=CrouchTargetedReaction
    via=AnimStateTransitionNode_3
   }]
 }
 CrouchTargetedReaction {
  path=MovementMode_SourcePose.AnimGraphNode_StateMachine_0.CrouchTargetedState.AnimStateNode_1.CrouchTargetedReaction
  nodes [Output Animation Pose(Result=Apply Additive(
     Base=Use cached pose 'InSourcePose'()
     Additive="Sequence Player (additive) 'Pistol_AutoAim_TargetedReact_CMM'"() {bLoopAnimation=false, sequence=&/Pistol_AutoAim_TargetedReact_CMM}
    ) {alphaBoolBlend {blendOption=Linear}})]
 }
 CrouchTargetedSourcePose {
  path=MovementMode_SourcePose.AnimGraphNode_StateMachine_0.CrouchTargetedState.AnimStateNode_3.CrouchTargetedSourcePose
  nodes [Output Animation Pose(Result=Use cached pose 'InSourcePose'())]
 }
 Transition {path=MovementMode_SourcePose.AnimGraphNode_StateMachine_0.CrouchTargetedState.AnimStateTransitionNode_3.Transition, nodes [Result()]}
 TargetedState {
  entry=TargetedReaction
  path=MovementMode_SourcePose.AnimGraphNode_StateMachine_4.TargetedState
  ids=id
  nodes [
   AnimStateNode_1 = TargetedReaction() {is=title AnimStateNode, bAlwaysResetOnEntry=true, stateLeft {guid=#1, notifyName=Targeted_Out}}
   AnimStateNode_3 = TargetedSourcePose() {is=title AnimStateNode}
   AnimStateTransitionNode_3 = AnimStateTransitionNode() {bAutomaticRuleBasedOnSequencePlayerInState=true}
  ]
  links [{
    dst=TargetedSourcePose
    rule=MovementMode_SourcePose.AnimGraphNode_StateMachine_4.TargetedState.AnimStateTransitionNode_3.Transition
    src=TargetedReaction
    via=AnimStateTransitionNode_3
   }]
 }
 TargetedReaction {
  path=MovementMode_SourcePose.AnimGraphNode_StateMachine_4.TargetedState.AnimStateNode_1.TargetedReaction
  nodes [Output Animation Pose(Result=Apply Additive(
     Base=Use cached pose 'InSourcePose'()
     Additive="Sequence Player (additive) 'Pistol_AutoAim_TargetedReact_CMM'"() {bLoopAnimation=false, sequence=&Pistol_AutoAim_TargetedReact_CMM}
    ) {alphaBoolBlend {blendOption=Linear}})]
 }
 TargetedSourcePose {
  path=MovementMode_SourcePose.AnimGraphNode_StateMachine_4.TargetedState.AnimStateNode_3.TargetedSourcePose
  nodes [Output Animation Pose(Result=Use cached pose 'InSourcePose'())]
 }
 Transition {path=MovementMode_SourcePose.AnimGraphNode_StateMachine_4.TargetedState.AnimStateTransitionNode_3.Transition, nodes [Result()]}
]
variables []
