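; Player pawn component for the Smartgun auto-aim pistol, which builds stacking lock-ons while aiming and spends them in timed bursts.
; On the local client, tick picks a reticle target with loss forgiveness, then adds locks faster nearer reticle center.
; Remaining lock count selects each shot's target socket.
; The target goes to the server and replicates back, driving the victim's camera effect and warning sound behind a console variable.
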(event EventBeginPlay
  (Components|Tick|SetComponentTickEnabled (not (Networking|IsServer)))
  (bind _asfort_player_pawn (Utilities|Casting|CastToFortPlayerPawn (Components|GetOwner)))
  (EventDispatchers|BindEventtoOnDBNOStateChanged _asfort_player_pawn (delegate Custom|DBNOChanged)))

(event Custom|TellServerLockOn
  (params (NewLockTarget :type Fort%20Pawn%20Object%20Reference))
  (bind _output_get (Variables|Default|SetRepLockTarget NewLockTarget))
  (Networking|ForceNetUpdate _output_get)
  (Variables|Default|GetRepLockTarget
    (:then
      (Default|CallOnGainedLockTarget))
    (:else
      (Default|CallOnLostLockTarget))))

(event Custom|DBNOChanged
  (params (FortPawn :type Fort%20Pawn%20Object%20Reference) (bInIsDBNO :type Boolean))
  (CallFunction|RemoveTargetUI))

(event Custom|MulticastTargeting
  (params (IsTargeting :type Boolean) (InWeapon :type Actor%20Object%20Reference))
  (if IsTargeting
    (Projectile|AddGameplayCue%28Looping%29Local InWeapon (Variables|Default|GetWeaponTargetingTag) (Ability|GameplayCue|MakeGameplayCueParameters :NormalizedMagnitude 0.0 :RawMagnitude 0.0 :Location "0,0,0" :Normal "0,0,0" :Instigator (Components|GetOwner) :GameplayEffectLevel 1 :AbilityLevel 1 :bReplicateLocationWhenUsingMinimalRepProxy false) false)
    (bind _asb_auto_aim_pistol (Utilities|Casting|CastToB_AutoAim_Pistol InWeapon))
    (Variables|Default|SetCachedWeapon _asb_auto_aim_pistol)
    (else
      (Projectile|RemoveGameplayCue%28Looping%29Local InWeapon (Variables|Default|GetWeaponTargetingTag) (Ability|GameplayCue|MakeGameplayCueParameters :NormalizedMagnitude 0.0 :RawMagnitude 0.0 :Location "0,0,0" :Normal "0,0,0" :Instigator (Components|GetOwner) :GameplayEffectLevel 1 :AbilityLevel 1 :bReplicateLocationWhenUsingMinimalRepProxy false)))))

(event Custom|TellServerTargeting
  (params (IsTargeting :type Boolean) (InWeapon :type Actor%20Object%20Reference))
  (CallFunction|MulticastTargeting IsTargeting InWeapon))

(event EventTick
  (params (DeltaSeconds :type Float%20%28single-precision%29))
  (CallFunction|TickVictimWarning)
  (bind _equipped (CallFunction|CanTickLocalShooter))
  (if _equipped
    (CallFunction|TickReticle DeltaSeconds)
    (CallFunction|TickLockOns DeltaSeconds)))

(event EventEndPlay
  (params (EndPlayReason :type EEndPlayReason%20Enum :default "Destroyed"))
  (CallFunction|RemoveTargetUI)
  (Projectile|RemoveGameplayCue%28Looping%29Local (Variables|Default|GetCachedWeapon) (Variables|Default|GetWeaponTargetingTag) (Ability|GameplayCue|MakeGameplayCueParameters :NormalizedMagnitude 0.0 :RawMagnitude 0.0 :Location "0,0,0" :Normal "0,0,0" :Instigator (Components|GetOwner) :GameplayEffectLevel 1 :AbilityLevel 1 :bReplicateLocationWhenUsingMinimalRepProxy false))
  (bind _spawnedwarningsound (Variables|Default|GetSpawnedWarningSound)
    (:then
      (Audio|Components|Audio|Stop _spawnedwarningsound))
    (:else
      _))
  (Utilities|IsValid (Variables|Default|GetLocalLockTarget)
    (:Is%20Valid
      (bind _outmodifier (Class|WeaponFunctionLibrary|FindorAddCameraModifierforPawn "/Smartgun/Gameplay/StackingLockons/ScreenFX/CameraModifier_AutoAimVictim.CameraModifier_AutoAimVictim_C" (Variables|Default|GetLocalLockTarget)))
      (bind _ascamera_modifier_auto_aim_victim (Utilities|Casting|CastToCameraModifier_AutoAimVictim _outmodifier))
      (Variables|/Smartgun/Gameplay/StackingLockons/ScreenFX/CameraModifier_AutoAimVictim.CameraModifier_AutoAimVictim_C|SetIsActive false _ascamera_modifier_auto_aim_victim)
      (Variables|Default|SetWarningActive false))))

(fn CanFireLockedOn
  (bind _cachedweapon (Variables|Default|GetCachedWeapon)
    (:then
      (return (or (== (FortWeapon|GetLocalMagazineAmmoCount _cachedweapon) (Variables|AutoAimWeaponPawnComponent|GetCurrentLockOnCount)) (>= (Variables|AutoAimWeaponPawnComponent|GetCurrentLockOnCount) 1))))
    (:else
      (return false))))

(fn CanFreeFire
  (return (>= (- (Utilities|Time|GetTimeSeconds) (Variables|Default|GetLastAmmoConsumeTime)) 0.1)))

(fn CanLockOn
  (bind _cachedweapon (Variables|Default|GetCachedWeapon)
    (:then
      (if (Math|Boolean|ANDBoolean (not (FortWeapon|IsReloading _cachedweapon)) (Camera|GetIsTargeting _cachedweapon) (Utilities|IsValid (Variables|Default|GetAutoAimTarget) :!class /Script/Engine.KismetSystemLibrary) (> (- (Utilities|Time|GetTimeSeconds) (Variables|Default|GetLastAmmoConsumeTime)) (ScalableFloat|GetValueatLevel (Variables|Default|GetAfterFiringLockDelay))))
        (bind _isallowedtolock (Default|IsAllowedtoLock (Variables|AutoAimWeaponPawnComponent|GetLockOnTarget)))
        (return _isallowedtolock)
        (else
          (return false))))
    (:else
      (return false))))

(fn CanTickLocalShooter
  (Utilities|FlowControl|Sequence
    (:then_0
      (Variables|Default|SetCachedWeapon 0)
      (bind _asfort_player_pawn (Utilities|Casting|CastToFortPlayerPawn (Components|GetOwner)))
      (if (Pawn|IsLocallyControlled _asfort_player_pawn)
        (bind _asb_auto_aim_pistol (Utilities|Casting|CastToB_AutoAim_Pistol (Weapon|GetCurrentWeapon _asfort_player_pawn)))
        (Variables|Default|SetCachedWeapon _asb_auto_aim_pistol)
        (return true)))
    (:then_1
      (return false))))

(fn ClearAutoAimTarget
  (Utilities|IsValid (Variables|Default|GetAutoAimTarget)
    (:Is%20Valid
      (Default|CallOnLostLockTarget)
      (CallFunction|RemoveTargetUI)
      (Variables|Default|SetAutoAimTarget 0)
      (if (> (Variables|AutoAimWeaponPawnComponent|GetCurrentLockOnCount) 0)
        (Audio|PlaySound2D "/Smartgun/Gameplay/StackingLockons/Sounds/AutoAim_Target_Lost_Cue.AutoAim_Target_Lost_Cue"))
      (Variables|AutoAimWeaponPawnComponent|SetCurrentLockOnCount 0)
      (Variables|AutoAimWeaponPawnComponent|SetProgressTowardNextLockOn 0.0)
      (CallFunction|TellServerLockOn 0))))

(fn FireBurst
  (bind _outcanfire (CallFunction|CanFireLockedOn))
  (if _outcanfire
    (Variables|Default|SetNumBurstShots 0)
    (Variables|Default|SetBurstSize (Variables|AutoAimWeaponPawnComponent|GetCurrentLockOnCount))
    (CallFunction|TrytoFireLockedOn)
    (CallFunction|RemoveTargetUI)))

(fn GetAngleToTarget)

(fn GetBulletTargetLocation
  (bind _autoaimtarget (Variables|Default|GetAutoAimTarget))
  (bind _autoaimtarget_2 (Variables|Default|GetAutoAimTarget))
  (bind _capsulecomponent (Variables|/Script/Engine.Character|GetCapsuleComponent _autoaimtarget_2))
  (bind _mesh (Variables|/Script/Engine.Character|GetMesh _autoaimtarget))
  (Utilities|FlowControl|Sequence
    (:then_0
      (Utilities|IsValid _autoaimtarget
        (:Is%20Valid
          (Utilities|FlowControl|Sequence
            (:then_0
              (Utilities|Casting|CastToNPC_Pawn_Wildlife_Parent _autoaimtarget
                (:then
                  (Utilities|IsValid _capsulecomponent
                    (:Is%20Valid
                      (return (Physics|GetCenterOfMass _capsulecomponent "None")))))
                (:CastFailed
                  (Utilities|IsValid _mesh
                    (:Is%20Valid
                      (switch int (Variables|AutoAimWeaponPawnComponent|GetCurrentLockOnCount)
                        (:1
                          (if (== (Variables|Default|GetBurstSize) 4)
                            (return (Transformation|GetSocketLocation _mesh "FX_Head"))
                            (else
                              (return (Transformation|GetSocketLocation _mesh "FX_Chest")))))
                        (:2
                          (return (Transformation|GetSocketLocation _mesh "SM_R_Shoulders")))
                        (:3
                          (return (Transformation|GetSocketLocation _mesh "FX_Chest")))
                        (:4
                          (return (Transformation|GetSocketLocation _mesh "SM_Pelvis")))))))))
            (:then_1
              ;[1 0.4635 0.4635] We have a valid target, but did not find a good output location. Return the target actor's location.
              (return (Transformation|GetActorLocation _autoaimtarget)))))))
    (:then_1
      ;[1 0.4635 0.4635] Complete failure, return zero vector.
      (return "0, 0, 0"))))

(fn IsAllowedToLock
  (params (LockOnTarget :type Fort%20Pawn%20Object%20Reference))
  (if (FortPawn|IsDead LockOnTarget)
    (return false)
    (else
      (if (GameplayTags|HasAnyMatchingGameplayTags LockOnTarget (Variables|Default|GetOverride%20Lock%20On%20Tag%20Container))
        (return true)
        (else
          (return (not (GameplayTags|HasAnyMatchingGameplayTags LockOnTarget (Variables|Default|GetPrevent%20Lock%20On%20Tag%20Container)))))))))

(fn OnConsumeAmmo
  (params (Quantity :type Integer))
  (Variables|Default|SetLastAmmoConsumeTime (Utilities|Time|GetTimeSeconds)))

(fn OnReload
  (Variables|AutoAimWeaponPawnComponent|SetCurrentLockOnCount 0))

(fn OnRep_RepLockTarget
  (bind _replocktarget (Variables|Default|GetRepLockTarget))
  (if (not (Networking|IsServer))
    (bind _returnvalue (Development|GetConsoleVariableBoolValue "SmartGun.EnableBeingLockedOnWidget"))
    (if _returnvalue
      (if (Variables|Default|GetWarningActive)
        (Utilities|IsValid _replocktarget
          (:Is%20Valid
            (if (!= (Variables|Default|GetLocalLockTarget) _replocktarget)
              _))
          (:Is%20Not%20Valid
            _))
        (bind _outmodifier (Class|WeaponFunctionLibrary|FindorAddCameraModifierforPawn "/Smartgun/Gameplay/StackingLockons/ScreenFX/CameraModifier_AutoAimVictim.CameraModifier_AutoAimVictim_C" (Variables|Default|GetLocalLockTarget)))
        (bind _ascamera_modifier_auto_aim_victim (Utilities|Casting|CastToCameraModifier_AutoAimVictim _outmodifier))
        (bind _spawnedwarningsound (Variables|Default|GetSpawnedWarningSound)
          (:then
            (if (Audio|Components|Audio|IsPlaying _spawnedwarningsound)
              (Audio|Components|Audio|Stop _spawnedwarningsound)
              (else
                _))
            (Variables|/Smartgun/Gameplay/StackingLockons/ScreenFX/CameraModifier_AutoAimVictim.CameraModifier_AutoAimVictim_C|SetIsActive false _ascamera_modifier_auto_aim_victim)
            (Variables|Default|SetWarningActive false))
          (:else
            _))
        (else
          (Utilities|IsValid _replocktarget
            (:Is%20Valid
              (if (Pawn|IsLocallyViewed _replocktarget)
                (Variables|Default|SetWarningActive true)
                (Variables|Default|SetLocalLockTarget _replocktarget)
                (bind _outmodifier_2 (Class|WeaponFunctionLibrary|FindorAddCameraModifierforPawn "/Smartgun/Gameplay/StackingLockons/ScreenFX/CameraModifier_AutoAimVictim.CameraModifier_AutoAimVictim_C" _replocktarget))
                (bind _ascamera_modifier_auto_aim_victim_2 (Utilities|Casting|CastToCameraModifier_AutoAimVictim _outmodifier_2))
                (bind _spawnedwarningsound_2 (Variables|Default|GetSpawnedWarningSound)
                  (:then
                    (if (Audio|Components|Audio|IsPlaying _spawnedwarningsound_2)
                      (Audio|Components|Audio|Stop _spawnedwarningsound_2)
                      (else
                        _))
                    (bind _returnvalue_2 (Audio|SpawnSoundAttached "/Smartgun/Gameplay/StackingLockons/Sounds/AutoAim_Target_Lock_Enemy_Cue.AutoAim_Target_Lock_Enemy_Cue" (Variables|/Script/Engine.Actor|GetRootComponent (Components|GetOwner))))
                    (Variables|Default|SetSpawnedWarningSound _returnvalue_2)
                    (Variables|/Smartgun/Gameplay/StackingLockons/ScreenFX/CameraModifier_AutoAimVictim.CameraModifier_AutoAimVictim_C|SetIsActive true _ascamera_modifier_auto_aim_victim_2))
                  (:else
                    _))))))))))

(fn OnTriggerPress
  (bind _cachedweapon (Variables|Default|GetCachedWeapon)
    (:then
      (if (Math|Boolean|ANDBoolean (Camera|GetIsTargeting _cachedweapon) (> (Variables|AutoAimWeaponPawnComponent|GetCurrentLockOnCount) 0) (== (Variables|Default|GetNumBurstShots) 0))
        (CallFunction|FireBurst)
        (else
          _)))
    (:else
      _))
  (bind _outcanfire (CallFunction|CanFreeFire))
  (if _outcanfire
    (Variables|Default|SetNumBurstShots 0)
    (bind _cachedweapon_2 (Variables|Default|GetCachedWeapon)
      (:then
        (Class|FortWeapon|TrytoCustomFire _cachedweapon_2 false))
      (:else))))

(fn OnUnequip
  (CallFunction|ClearAutoAimTarget))

(fn RemoveTargetUI
  (Utilities|IsValid (Variables|Default|GetAutoAimTarget)
    (:Is%20Valid
      (Projectile|RemoveGameplayCue%28Looping%29Local (Variables|Default|GetAutoAimTarget) (Variables|Default|GetLockTargetCueTag) (Ability|GameplayCue|MakeGameplayCueParameters 0.0)))))

(fn SetMaterialLockonParams
  (params (Count :type Integer))
  (bind _cachedweapon (Variables|Default|GetCachedWeapon)
    (:then
      (Utilities|IsValid (Variables|/Smartgun/Gameplay/StackingLockons/B_AutoAim_Pistol.B_AutoAim_Pistol_C|GetMID _cachedweapon)
        (:Is%20Valid
          (Rendering|Material|SetScalarParameterValue (Variables|/Smartgun/Gameplay/StackingLockons/B_AutoAim_Pistol.B_AutoAim_Pistol_C|GetMID _cachedweapon) "Lockon_Progress" (Math|Conversions|ToFloat(Integer) Count) :!class /Script/Engine.MaterialInstanceDynamic)
          (if (== Count (ScalableFloat|GetValueasInteger (Variables|AutoAimWeaponPawnComponent|GetMaxLockOns)))
            (if (not (Variables|Default|GetMID_MaxLock))
              (Variables|Default|SetMID_MaxLock true)
              (Rendering|Material|SetScalarParameterValue (Variables|/Smartgun/Gameplay/StackingLockons/B_AutoAim_Pistol.B_AutoAim_Pistol_C|GetMID _cachedweapon) "Lockon_Blink" (Utilities|Time|GetGameTimeinSeconds) :!class /Script/Engine.MaterialInstanceDynamic))
            (else
              (Variables|Default|SetMID_MaxLock false))))))
    (:else)))

(fn SmartPDWComponent_2_AutoGenFunc)

(fn TargetingChanged
  (params (NewTargeting :type Boolean))
  (CallFunction|TellServerTargeting NewTargeting (Variables|Default|GetCachedWeapon))
  (bind _cachedweapon (Variables|Default|GetCachedWeapon)
    (:then
      (Reticle|SetShouldDrawNativeReticle _cachedweapon (not NewTargeting)))
    (:else)))

(fn TickLockOns
  (params (Delta%20Seconds :type Float%20%28double-precision%29))
  (bind _cachedweapon (Variables|Default|GetCachedWeapon)
    (:then
      (if (Camera|GetIsTargeting _cachedweapon)
        (if (< (Variables|AutoAimWeaponPawnComponent|GetCurrentLockOnCount) (FortWeapon|GetLocalMagazineAmmoCount (Variables|Default|GetCachedWeapon)))
          (Variables|AutoAimWeaponPawnComponent|SetTargetToReticleDistanceNormalized (Utilities|RemapFloat (Math|Float|Clamp(Float) (Math|Float|SafeDivide (Variables|Default|GetReticleDistance) (Variables|Default|GetInnerReticleSize))) "(X=0.100000,Y=1.000000)" "(X=0.000000,Y=1.000000)" true))
          (if (Utilities|IsValid (Variables|Default|GetAutoAimTarget) :!class /Script/Engine.KismetSystemLibrary)
            (if (< (Variables|AutoAimWeaponPawnComponent|GetCurrentLockOnCount) (ScalableFloat|GetValueasInteger (Variables|AutoAimWeaponPawnComponent|GetMaxLockOns)))
              (bind _outcanlock (CallFunction|CanLockOn))
              (if (and _outcanlock (not (Variables|Default|GetWasLockAllowed)))
                (Default|CallOnLockonAllowed)
                (else
                  _))
              (bind _output_get (Variables|Default|SetWasLockAllowed _outcanlock))
              (if _output_get
                (if (Utilities|IsValid (Variables|Default|GetAutoAimTarget) :!class /Script/Engine.KismetSystemLibrary)
                  (Variables|AutoAimWeaponPawnComponent|SetProgressTowardNextLockOn (+ (Variables|AutoAimWeaponPawnComponent|GetProgressTowardNextLockOn) (Variables|GetDelta%20Seconds)))
                  (if (>= (Variables|AutoAimWeaponPawnComponent|GetProgressTowardNextLockOn) (Math|Float|Lerp (ScalableFloat|GetValueatLevel (Variables|AutoAimWeaponPawnComponent|GetLockOnTimeReticleCenter)) (ScalableFloat|GetValueatLevel (Variables|AutoAimWeaponPawnComponent|GetLockOnTimeReticleEdge)) (Variables|AutoAimWeaponPawnComponent|GetTargetToReticleDistanceNormalized)))
                    (Math|Integer|IncrementInt (Variables|AutoAimWeaponPawnComponent|GetCurrentLockOnCount)
                      (:%20%20
                        (Variables|AutoAimWeaponPawnComponent|SetProgressTowardNextLockOn 0.0)
                        (if (not (Variables|Default|GetTargetHasLockUI))
                          (Projectile|AddGameplayCue%28Looping%29Local (Variables|Default|GetAutoAimTarget) (Variables|Default|GetLockTargetCueTag) (Ability|GameplayCue|MakeGameplayCueParameters :NormalizedMagnitude 0.0 :RawMagnitude 0.0 :Location "0,0,0" :Normal "0,0,0" :Instigator (Components|GetOwner) :GameplayEffectLevel 1 :AbilityLevel 1 :bReplicateLocationWhenUsingMinimalRepProxy false)))
                        (if (== (Variables|AutoAimWeaponPawnComponent|GetCurrentLockOnCount) (ScalableFloat|GetValueasInteger (Variables|AutoAimWeaponPawnComponent|GetMaxLockOns)))
                          (Audio|PlaySound2D "/Smartgun/Gameplay/StackingLockons/Sounds/AutoAim_Target_Lock_Lock_Cue.AutoAim_Target_Lock_Lock_Cue")
                          (Default|CallOnFullCharge)
                          (else
                            (Audio|PlaySound2D "/Smartgun/Gameplay/StackingLockons/Sounds/AutoAim_Target_Count_Cue.AutoAim_Target_Count_Cue")
                            (if (== (Variables|AutoAimWeaponPawnComponent|GetCurrentLockOnCount) 1)
                              (Audio|PlaySound2D "/Smartgun/Gameplay/StackingLockons/Sounds/AutoAim_Target_Lock_Appears_Cue.AutoAim_Target_Lock_Appears_Cue")))))))
                  (else
                    (Variables|AutoAimWeaponPawnComponent|SetProgressTowardNextLockOn 0.0)
                    (Variables|AutoAimWeaponPawnComponent|SetCurrentLockOnCount 0)))))
            (else
              (Variables|AutoAimWeaponPawnComponent|SetProgressTowardNextLockOn 0.0))))
        (else
          (Variables|AutoAimWeaponPawnComponent|SetCurrentLockOnCount 0)
          (Variables|Default|SetNumBurstShots 0))))
    (:else)))

(fn TickReticle
  (params (DeltaSeconds :type Float%20%28double-precision%29))
  (if (not (Networking|IsServer))
    (bind _cachedweapon (Variables|Default|GetCachedWeapon)
      (:then
        (if (Camera|GetIsTargeting _cachedweapon)
          (bind _asfort_player_pawn (Utilities|Casting|CastToFortPlayerPawn (Components|GetOwner)))
          (LocalVariables|SetPawn _asfort_player_pawn)
          (CallFunction|SetMaterialLockonParams (Variables|AutoAimWeaponPawnComponent|GetCurrentLockOnCount))
          (if (Utilities|IsValid (Variables|Default|GetAutoAimTarget) :!class /Script/Engine.KismetSystemLibrary)
            ;[0 0.1864 1] Clear auto aim target when no longer valid
            (bind _isallowedtolock (Default|IsAllowedtoLock (Variables|Default|GetAutoAimTarget)))
            (if _isallowedtolock
              (bind (_outtargetpawn _outreticledistance) (AutoAimWeapon|FindBestAutoAimTarget (LocalVariables|GetPawn) (Variables|Default|GetOuterReticleSize) 5e+03))
              (Variables|Default|SetReticleDistance _outreticledistance)
              (Utilities|IsValid _outtargetpawn
                (:Is%20Valid
                  ;[1 1 1] Target Switch - Only if the old target left the reticle
                  (if (!= _outtargetpawn (Variables|Default|GetAutoAimTarget))
                    (Utilities|IsValid _outtargetpawn
                      (:Is%20Valid
                        (bind (_outreticledistance_2 _returnvalue) (AutoAimWeapon|DoesAutoAimWeaponReticleIntersectTarget (LocalVariables|GetPawn) (Variables|Default|GetAutoAimTarget) (Variables|Default|GetInnerReticleSize)))
                        (if _returnvalue
                          (bind _returnvalue_2 (AutoAimWeapon|DoesAutoAimWeaponHaveLineOfSight (LocalVariables|GetPawn) (Variables|Default|GetAutoAimTarget)))
                          (if _returnvalue_2
                            _
                            (else
                              _))
                          (else
                            _))
                        (Variables|AutoAimWeaponPawnComponent|SetCurrentLockOnCount 0)
                        (Projectile|RemoveGameplayCue%28Looping%29Local (Variables|Default|GetAutoAimTarget) (Variables|Default|GetLockTargetCueTag) (Ability|GameplayCue|MakeGameplayCueParameters 0.0))
                        (Variables|Default|SetAutoAimTarget _outtargetpawn)
                        (Variables|Default|SetTimeSinceLostTarget 0.0)
                        (CallFunction|TellServerLockOn (Variables|Default|GetAutoAimTarget))
                        (Projectile|AddGameplayCue%28Looping%29Local (Variables|Default|GetAutoAimTarget) (Variables|Default|GetLockTargetCueTag) (Ability|GameplayCue|MakeGameplayCueParameters :NormalizedMagnitude 0.0 :RawMagnitude 0.0 :Location "0,0,0" :Normal "0,0,0" :Instigator (Components|GetOwner) :GameplayEffectLevel 1 :AbilityLevel 1 :bReplicateLocationWhenUsingMinimalRepProxy false))))
                    (else
                      (Variables|Default|SetTimeSinceLostTarget 0.0))))
                (:Is%20Not%20Valid
                  ;[1 1 1] Lost Target
                  (bind _output_get (Variables|Default|SetTimeSinceLostTarget (+ (Variables|Default|GetTimeSinceLostTarget) (Variables|GetDeltaSeconds))))
                  (bind (_outreticledistance_3 _returnvalue_3) (AutoAimWeapon|DoesAutoAimWeaponReticleIntersectTarget (LocalVariables|GetPawn) (Variables|Default|GetAutoAimTarget) (Variables|Default|GetOuterReticleSize)))
                  (if (not (and _returnvalue_3 (< _output_get (ScalableFloat|GetValueatLevel (Variables|Default|GetLockOnForgivenessTime)))))
                    (CallFunction|ClearAutoAimTarget))))
              (else
                (CallFunction|ClearAutoAimTarget)
                (CallFunction|SetMaterialLockonParams 0)))
            (else
              ;[1 1 1] Looking for target
              (bind _outtargetpawn_2 (AutoAimWeapon|FindBestAutoAimTarget (LocalVariables|GetPawn) (Variables|Default|GetInnerReticleSize) (ScalableFloat|GetValueatLevel (Variables|Default|GetLockOnRange))))
              (Utilities|IsValid _outtargetpawn_2
                (:Is%20Valid
                  (Variables|Default|SetAutoAimTarget _outtargetpawn_2)
                  (Variables|Default|SetTimeSinceLostTarget 0.0)
                  (CallFunction|TellServerLockOn (Variables|Default|GetAutoAimTarget))
                  (Default|CallOnGainedLockTarget)))))
          (else
            _)))
      (:else
        _))
    (Variables|Default|GetAutoAimTarget
      (:then
        (CallFunction|ClearAutoAimTarget)
        (CallFunction|SetMaterialLockonParams 0))
      (:else))))

(fn TickVictimWarning
  (bind _replocktarget (Variables|Default|GetRepLockTarget))
  (if (and (Utilities|IsValid (Variables|Default|GetSpawnedWarningSound) :!class /Script/Engine.KismetSystemLibrary) (not (Utilities|IsValid (Variables|Default|GetRepLockTarget) :!class /Script/Engine.KismetSystemLibrary)))
    (Audio|Components|Audio|Stop (Variables|Default|GetSpawnedWarningSound)))
  (Utilities|IsValid _replocktarget
    (:Is%20Valid
      (if (or (FortPawn|IsDead _replocktarget) (Health|IsDBNO _replocktarget))
        (Variables|Default|SetWarningActive false)
        (bind _outmodifier (Class|WeaponFunctionLibrary|FindorAddCameraModifierforPawn "/Smartgun/Gameplay/StackingLockons/ScreenFX/CameraModifier_AutoAimVictim.CameraModifier_AutoAimVictim_C" (Variables|Default|GetLocalLockTarget)))
        (bind _ascamera_modifier_auto_aim_victim (Utilities|Casting|CastToCameraModifier_AutoAimVictim _outmodifier))
        (Variables|/Smartgun/Gameplay/StackingLockons/ScreenFX/CameraModifier_AutoAimVictim.CameraModifier_AutoAimVictim_C|SetIsActive false _ascamera_modifier_auto_aim_victim)))
    (:Is%20Not%20Valid
      (Variables|Default|SetWarningActive false))))

(fn TryToFireLockedOn
  (bind _autoaimtarget (Variables|Default|GetAutoAimTarget))
  (if (and (> (Variables|AutoAimWeaponPawnComponent|GetCurrentLockOnCount) 0) (< (Variables|Default|GetNumBurstShots) 4))
    (bind _cachedweapon (Variables|Default|GetCachedWeapon)
      (:then
        (Class|FortWeapon|TrytoCustomFire _cachedweapon false)
        (Math|Integer|IncrementInt (Variables|Default|GetNumBurstShots)
          (:%20%20
            (Utilities|IsValid _autoaimtarget
              (:Is%20Valid
                (if (not (FortPawn|IsDead _autoaimtarget))
                  (Variables|AutoAimWeaponPawnComponent|SetCurrentLockOnCount (Math|Integer|Max(Integer) (- (Variables|AutoAimWeaponPawnComponent|GetCurrentLockOnCount) 1)))
                  (Utilities|Time|SetTimerbyEvent (EventDispatchers|CreateEvent :!event Try%20to%20Fire%20Locked%20On) 0.09)))))))
      (:else))
    (else
      (Variables|Default|SetNumBurstShots 0))))
