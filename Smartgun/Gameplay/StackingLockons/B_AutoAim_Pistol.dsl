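; Auto-aim pistol weapon Blueprint for the Smartgun, which steers shots toward locked-on targets through a cached AutoAim pistol component on the instigator.
; Targeting, trigger, ammo, reload and unequip events forward to that component, whose lock gained/lost events toggle the override downsights spread.
; An aim-rotation override turns camera aim toward the component's bullet target when locked on, otherwise optionally applies random cone spread.
; Tracer beams receive endpoints and targeting state as Niagara parameters.
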(event EventOnWrapApplied
  (params (WrapDefinition :type Athena%20Item%20Wrap%20Definition%20Object%20Reference))
  ;[1 1 1] re updating dynamic material instance after wrap is applyed.
  (bind _returnvalue (Fortnite|GetDynamicMaterialInstance (Variables|Visual|GetWeaponMesh) 1))
  (Variables|Default|SetMID _returnvalue))

(event EventBeginPlay
  (CallFunction|GetSmartComponentCached)
  (EventDispatchers|BindEventtoOnLocalTargetingChanged (EventDispatchers|CreateEvent :!event Custom|TargetingChanged))
  (EventDispatchers|BindEventtoOnPressTrigger (EventDispatchers|CreateEvent :!event Custom|TriggerPressed))
  (EventDispatchers|BindEventtoOnConsumeAmmo (EventDispatchers|CreateEvent :!event Custom|OnConsumedAmmo))
  (EventDispatchers|BindEventtoOnLocalReloadStarted (EventDispatchers|CreateEvent :!event Custom|ReloadStarted))
  (Variables|FortWeapon|SetOnGetAimRotOverride (EventDispatchers|CreateEvent :!event OnGetAimRot))
  (Variables|FortWeapon|SetOnBeamFiredDelegate (EventDispatchers|CreateEvent :!event Custom|OnTracerBeamFired)))

(event EventEndPlay
  (params (EndPlayReason :type EEndPlayReason%20Enum :default "Destroyed"))
  (bind _smartcomp (Variables|Default|GetSmartComp)
    (:then
      (Default|UnbindEventfromOnGainedLockTarget _smartcomp (EventDispatchers|CreateEvent :!event Custom|HandleTargetAcquired))
      (Default|UnbindEventfromOnLostLockTarget _smartcomp (EventDispatchers|CreateEvent :!event Custom|HandleTargetLost)))
    (:else)))

(event Custom|TargetingChanged
  (params (bIsTargeting :type Boolean))
  (bind _component (CallFunction|GetSmartComponentCached))
  (Utilities|IsValid _component
    (:Is%20Valid
      (Class|AutoAimPistolComponent|TargetingChanged _component bIsTargeting))))

(event Custom|TriggerPressed
  (params (Weapon :type Fort%20Weapon%20Object%20Reference))
  (bind _component (CallFunction|GetSmartComponentCached))
  (Utilities|IsValid _component
    (:Is%20Valid
      (Class|AutoAimPistolComponent|OnTriggerPress _component))))

(event Custom|OnConsumedAmmo
  (params (Quantity :type Integer))
  (bind _component (CallFunction|GetSmartComponentCached))
  (Utilities|IsValid _component
    (:Is%20Valid
      (Class|AutoAimPistolComponent|OnConsumeAmmo _component Quantity))))

(event Custom|ReloadStarted
  (params (ReloadTime :type Float%20%28single-precision%29) (ReloadType :type EFortWeaponReloadType%20Enum :default "ReloadWholeClip"))
  (bind _component (CallFunction|GetSmartComponentCached))
  (Utilities|IsValid _component
    (:Is%20Valid
      (Class|AutoAimPistolComponent|OnReload _component))))

(event FortWeapon|EventOnLocalUnEquip
  (bind _component (CallFunction|GetSmartComponentCached))
  (Utilities|IsValid _component
    (:Is%20Valid
      (Class|AutoAimPistolComponent|OnUnequip _component))))

(event Custom|OnTracerBeamFired
  (params (Weapon :type Fort%20Weapon%20Object%20Reference) (bUsePersistentBeam :type Boolean) (PSC :type FXSystem%20Component%20Object%20Reference) (BeamOrigin :type Vector%20%28by%20ref%29) (HitLocation :type Vector%20%28by%20ref%29))
  (bind _asniagara_particle_system_component (Utilities|Casting|CastToNiagaraComponent PSC))
  (Niagara|SetNiagaraVariable(Vector3) _asniagara_particle_system_component "BeamStart_0_0" BeamOrigin)
  (Niagara|SetNiagaraVariable(Vector3) _asniagara_particle_system_component "BeamEnd_0" HitLocation)
  (Niagara|SetNiagaraVariable(Vector3) _asniagara_particle_system_component "Direction" (Math|Vector|GetRightVector (Transformation|GetWorldRotation (Variables|Visual|GetWeaponMesh))))
  (Niagara|SetNiagaraVariable(Bool) _asniagara_particle_system_component "IsTargeting" (Camera|GetIsTargeting)))

(event Custom|HandleTargetAcquired
  ;[0 0.1106 1] Use override downsights spread (if valid) when locked on to a target.
  (FortWeapon|SetUseOverrideSpreadDownsightsMultiplier true))

(event Custom|HandleTargetLost
  ;[0 0.1106 1] Use override downsights spread (if valid) when locked on to a target.
  (FortWeapon|SetUseOverrideSpreadDownsightsMultiplier false))

(fn ApplySpreadToAimRotation
  (bind _asfort_pawn (Utilities|Casting|CastToFortPawn (Variables|Actor|GetInstigator)))
  (return (Math|Conversions|RotationFromXVector (Math|Random|RandomUnitVectorinConeinDegrees (Math|Rotator|GetRotationXVector (.rotation (Targeting|GetTargetingTransform _asfort_pawn "Camera" 0))) (ScalableFloat|GetValueatLevel (Variables|Spread|GetADSSpread))))))

(fn GetSmartComponentCached
  (bind _instigator (Variables|Actor|GetInstigator))
  (bind _returnvalue (Actor|GetComponentbyClass _instigator "/Smartgun/Gameplay/StackingLockons/AutoAim_Pistol_Component.AutoAim_Pistol_Component_C"))
  (bind _smartcomp (Variables|Default|GetSmartComp)
    (:then
      (return _smartcomp))
    (:else
      (Utilities|IsValid _instigator
        (:Is%20Valid
          (Utilities|IsValid _returnvalue
            (:Is%20Valid
              (bind _output_get (Variables|Default|SetSmartComp _returnvalue))
              (Default|BindEventtoOnGainedLockTarget _output_get (EventDispatchers|CreateEvent :!event HandleTargetAcquired))
              (Default|BindEventtoOnLostLockTarget _output_get (EventDispatchers|CreateEvent :!event HandleTargetLost))
              (return _output_get))
            (:Is%20Not%20Valid
              _)))
        (:Is%20Not%20Valid
          _))
      (return 0))))

(fn HasLockOns
  (return (> (Class|AutoAimWeaponPawnComponent|GetCurrentLockonCount (Variables|Default|GetSmartComp) :!class /Script/AutoAimWeaponRuntime.AutoAimWeaponPawnComponent) 3)))

(fn IsAimedAtAutoAimTarget
  (bind (_location _rotation) (Pawn|GetPlayerViewPoint (Fortnite|GetFortPlayerControllerfromActor (Variables|Actor|GetInstigator))))
  (Utilities|IsValid (Variables|/Smartgun/Gameplay/StackingLockons/AutoAim_Pistol_Component.AutoAim_Pistol_Component_C|GetAutoAimTarget (Variables|Default|GetSmartComp))
    (:Is%20Valid
      (bind _outhit (Collision|LineTraceByChannel _location (+ _location (* (Math|Vector|GetForwardVector _rotation) 5e+03)) "TraceTypeQuery6" false (Utilities|Array|MakeArray (Variables|Actor|GetInstigator))))
      (return (and (Collision|BreakHitResult _outhit) (== (.HitActor (Collision|BreakHitResult _outhit)) (Variables|/Smartgun/Gameplay/StackingLockons/AutoAim_Pistol_Component.AutoAim_Pistol_Component_C|GetAutoAimTarget (Variables|Default|GetSmartComp))))))))

(fn OnTriggerPressed
  (bind _outlockons (CallFunction|HasLockOns))
  (if _outlockons
    (CallFunction|TryToFire)))

(fn OnGetAimRot
  (params (Weapon :type Fort%20Weapon%20Object%20Reference) (TargetingSource :type EFortAbilityTargetingSource%20Enum :default "Camera"))
  (bind _returnvalue_location (Math|Transform|BreakTransform (Targeting|GetTargetingTransform (Utilities|Casting|CastToFortPawn (Variables|Actor|GetInstigator)) "Camera" 0)))
  (bind _instigator (Variables|Actor|GetInstigator))
  (bind (_asfort_pawn _bsuccess) (Utilities|Casting|CastToFortPawn _instigator))
  (bind _smartcomp (Variables|Default|GetSmartComp))
  (if (Math|Boolean|ANDBoolean (Utilities|Enum|Equal(Enum) TargetingSource "Camera") (Utilities|IsValid _smartcomp :!class /Script/Engine.KismetSystemLibrary) (Utilities|IsValid _instigator :!class /Script/Engine.KismetSystemLibrary))
    (if (Math|Boolean|ANDBoolean (Utilities|IsValid (Variables|/Smartgun/Gameplay/StackingLockons/AutoAim_Pistol_Component.AutoAim_Pistol_Component_C|GetAutoAimTarget _smartcomp) :!class /Script/Engine.KismetSystemLibrary) (> (Class|AutoAimWeaponPawnComponent|GetCurrentLockonCount _smartcomp :!class /Script/AutoAimWeaponRuntime.AutoAimWeaponPawnComponent) 0) _bsuccess)
      ;[0 0.5781 0.02235] Rotate aim toward target if locked on.
      (bind _outlocation (Class|AutoAimPistolComponent|GetBulletTargetLocation _smartcomp))
      (return (Math|Rotator|FindLookatRotation _returnvalue_location _outlocation) true)
      (elif (Variables|Spread|GetbUseBlueprintSpread)
        ;[1 0.3196 0] Apply additional spread if enabled and not locked on. This was here to add spread while aiming down sights and not locked on, since the Spread Downsights multiplier was otherwise 0. Can be disabled or removed now that we can set an Override Spread Downsights multiplier.
        (bind _outrotation (CallFunction|ApplySpreadtoAimRotation))
        (return _outrotation true)
        (else
          (return "0,0.000000,1.000000" false))))
    (else
      _)))

(fn TrytoFire)

(fn ConstructionScript
  (|CallParentFunction)
  (bind _returnvalue (Fortnite|GetDynamicMaterialInstance (Variables|Visual|GetWeaponMesh) 1))
  (Variables|Default|SetMID _returnvalue))
