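# Parent material for the auto-aim pistol, adding a segmented, blinking emissive readout of lock-on progress to the standard Fortnite weapon material.
# The weapon master feeds elemental and overheat effects (bypassed on mobile), the ADS scope function and optional vertex-color hiding.
# Lockon_Progress fills quantized UV1 bands through a sphere mask.
# A time-based sine blinks that glow for a set duration; it multiplies emissive before a time-of-day boost.
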Material /Smartgun/Gameplay/StackingLockons/Materials/M_AutoAim_Pistol_Parent
graph {
 properties {
  bAutomaticallySetUsageInEditor=false
  bUseMaterialAttributes=true
  bUsedWithClothing=true
  bUsedWithMeshDeformer=true
  bUsedWithMorphTargets=true
  bUsedWithNanite=true
  bUsedWithSkeletalMesh=true
  pixelDepthOffsetMode=PDOM_Legacy
  usedShadingModels=Default Lit
 }
 nodes [
  n1 = MaterialFunctionCall(In=NamedRerouteUsage()) {materialFunction=&MF_TechArtOverrides_accessories}
  MaterialFunctionCall(MA In=@n1.Result) {materialFunction=&MF_CreativeMoveTool}
  n2 = MaterialFunctionCall!(In) {materialFunction=/Game/Packages/Fortress_SharedMaterials/Base_Material_Functions/BaseCustomizationFunctions/MF_FN_Weapon_MASTER}
  n3 = GetMaterialAttributes(None=@n2.Material Result) {attributeGetTypes [#1]}
  n4 = VectorParameter() {group=Weapon Specific Wrap Adjustments, parameterName=DO NOT OVERRIDE-OBMAX, sortPriority=50, defaultValue=<rgba 0 0 0 1>}
  n5 = VectorParameter() {group=Weapon Specific Wrap Adjustments, parameterName=DO NOT OVERRIDE-OBMIN, sortPriority=51, defaultValue=<rgba 0 0 0 1>}
  n6 = VectorParameter() {group=Weapon Specific Wrap Adjustments, parameterName=DO NOT OVERRIDE-OBCENTER, sortPriority=55, defaultValue=<rgba 0 0 0 1>}
  n7 = VectorParameter() {group=Weapon Specific Wrap Adjustments, parameterName=DO NOT OVERRIDE-SZ-Back, sortPriority=100, defaultValue=<rgba 0.3 0 0 0.2>}
  n8 = VectorParameter() {group=Weapon Specific Wrap Adjustments, parameterName=DO NOT OVERRIDE-SZ-Middle, sortPriority=101, defaultValue=<rgba 0 0 0 0.2>}
  n9 = Add(A=@n7.RGB, B=@n8.RGB)
  n10 = VectorParameter() {group=Weapon Specific Wrap Adjustments, parameterName=DO NOT OVERRIDE-SZ-Front, sortPriority=102, defaultValue=<rgba -0.3 0 0 0.2>}
  n11 = VectorParameter() {group=Weapon Specific Wrap Adjustments, parameterName=DO NOT OVERRIDE-Virtual-Pivot, sortPriority=110, defaultValue=<rgba 0 0 0 0.5>}
  n12 = Add(A=Add(A=Add(A=@n4.RGB, B=@n5.RGB), B=@n6.RGB), B=Add(A=Add(A=@n9, B=@n10.RGB), B=@n11.RGB))
  n13 = SetMaterialAttributes!(
   MaterialAttributes=@n3.MaterialAttributes
   Base Color=LinearInterpolate(
    A=LinearInterpolate(A=@n3.Base Color, B=@n12) {constAlpha=0}
    B=ScalarParameter() {group=Weapon Specific Wrap Adjustments, parameterName=IsPickaxe, sortPriority=200}
   ) {constAlpha=0}
  ) {attributeSetTypes [#1]}
  n14 = MaterialFunctionCall!(
   Base MA=MaterialFunctionCall!(
    In=@n13
    Pre-Skinned Local Position=@n2.Pre-Skinned Local Position Pixel Shader
    Pre-Skinned Local Normal=@n2.Pre-Skinned Local Normal Pixel Shader
   ) {materialFunction=/Game/Weapons/WeaponSkins/MaterialFunction/ElementalWeaponsFunction}.Result
   Pre-Skinned Local Position=@n2.Pre-Skinned Local Position Pixel Shader
  ) {materialFunction=/Game/Weapons/WeaponSkins/MaterialFunction/OverheatGunFunction}
  n15 = GetMaterialAttributes(None=MaterialFunctionCall!(Material Attributes=ShadingPathSwitch(Default=@n14.Result, Mobile=@n13) {desc=Migrated from FeatureLevelSwitch}) {materialFunction=/Game/Packages/Fortnite_UniqueMaterials_1/Weapons/ADSScopeMaster}.Result) {attributeGetTypes [#2]}
  n16 = MaterialFunctionCall!(WPO) {materialFunction=/Game/Packages/Fortress_SharedMaterials/Base_Material_Functions/Fortnite_HideElementsWithVertexColor}
  NamedRerouteDeclaration(None=SetMaterialAttributes!(
    MaterialAttributes=@n15.MaterialAttributes
    World Position Offset=StaticSwitchParameter(True=@n16.Result, False=@n15.World Position Offset) {group=Opacity Mask, parameterName=Use Vertex Colors for Mask}
   ) {attributeSetTypes [#2]}) {name=Lockon_Attribute_In, nodeColor=<rgba 0 1 0.4141 1>}
  n17 = GetMaterialAttributes(None=NamedRerouteUsage()) {attributeGetTypes [#3]}
  n18 = Multiply(A=@n17.Emissive Color, B=NamedRerouteUsage())
  NamedRerouteDeclaration(None=SetMaterialAttributes!(
    MaterialAttributes=@n17.MaterialAttributes
    Emissive Color=MaterialFunctionCall!(In=@n18) {materialFunction=/Game/Weapons/WeaponSkins/MaterialFunction/BoostEmissivebyTOD}.Result
   ) {attributeSetTypes [#3]}) {name=Lockon_Attribute_OUT, nodeColor=<rgba 0.0078 1 0 1>}
  n19 = NamedRerouteUsage()
  n20 = ScalarParameter() {defaultValue=150, parameterName=Lockon_GlowIntensity}
  n21 = Divide(
   A=Subtract(A=Time() {period=2.213}, B=ScalarParameter() {defaultValue=-100, parameterName=Lockon_Blink})
   B=ScalarParameter() {defaultValue=0.7, parameterName=Lockon_BlinkDuration}
  )
  n22 = ScalarParameter() {defaultValue=3, parameterName=Lockon_NumOfBlinks}
  n23 = Ceil(None=Sine(None=Multiply(A=@n21, B=@n22)))
  n24 = Floor(None=Saturate(None=@n21))
  NamedRerouteDeclaration(None=Multiply(A=Multiply(A=@n19, B=@n20), B=LinearInterpolate(A=@n23, Alpha=@n24) {constA=1})) {name=Lockon, nodeColor=<rgba 0 1 0.0859 1>}
  n25 = ComponentMask(None=TextureCoordinate() {coordinateIndex=1}) {g=true}
  n26 = ScalarParameter() {defaultValue=5, parameterName=Lockon_Segments}
  n27 = Floor(None=Multiply(A=LinearInterpolate(Alpha=@n25) {constA=1, constB=0.25}, B=@n26))
  n28 = Divide(A=ScalarParameter() {parameterName=Lockon_Progress}, B=Subtract(A=@n26))
  NamedRerouteDeclaration(None=SphereMask(A=Divide(A=@n27, B=@n26) {constB=4}, B=Constant(), Radius=Add(A=@n28) {constB=0.001})) {name=Lockon_Progress, nodeColor=<rgba 1 0.3984 0 1>}
 ]
}
