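# Physics asset for the AutoAim pistol skeletal mesh in the Smartgun stacking lock-ons feature.
# It defines a single body on the handle bone with one capsule collision shape, angled to lie along the weapon.
# It has no constraints, so the pistol simulates and collides as one rigid piece.
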PhysicsAsset /Smartgun/Gameplay/StackingLockons/Meshes/AutoAim_Pistol_PhysicsAsset
bodies {handle [{shapeType=Capsule, length=17.1908, radius=10.7621, center=<11.0173 0.0009 2.5986>, rotation=<rot 0 89.9607 -84.4317>}]}
constraints []
