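# Sound cue for aiming down sights with the auto-aim Smartgun, part of its stacking lock-on gameplay, routed as weapon foley with weapon-reload attenuation.
# A mixer layers a pitched-up heavy pistol ADS-in sound with reticle-appear and reticle-counter sounds at individual volumes.
# A fourth, slightly delayed layer plays one of two pitched-down mechanical ADS variations, picked at random with equal weight.
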SoundCue /Smartgun/Gameplay/StackingLockons/Sounds/AutoAim_ADS_In_Cue
root=Mixer_0
nodes [
 {id=Delay_0, children [Modulator_1], properties {delayMax=0.15, delayMin=0.15}}
 {id=Mixer_0, children [Modulator_0,WavePlayer_0,WavePlayer_2,Delay_0], properties {inputVolume [0.6,2,0.75,1.5]}}
 {id=Modulator_0, children [WavePlayer_1], properties {pitchMax=1.1, pitchMin=1.1, volumeMax=0.7, volumeMin=0.7}}
 {id=Modulator_1, children [Random_0], properties {pitchMax=0.9, pitchMin=0.9}}
 {id=Random_0, children [WavePlayer_3,WavePlayer_4], properties {weights [1,1]}}
 {id=WavePlayer_0, properties {soundWaveAssetPtr=&/AutoAim_Target_Retical_Appears_02}}
 {id=WavePlayer_1, properties {soundWaveAssetPtr=/Game/Sounds/Fort_Weapons/ADS/Weapon_PISTOLHEAVY_ADS_IN_01}}
 {id=WavePlayer_2, properties {soundWaveAssetPtr=&/AutoAim_Retical_Counter_01}}
 {id=WavePlayer_3, properties {soundWaveAssetPtr=&/AutoAim_ADS_In_Mech_01}}
 {id=WavePlayer_4, properties {soundWaveAssetPtr=&/AutoAim_ADS_In_Mech_02}}
]
properties {
 attenuationSettings=&Fort_Weapon_Reload_Att
 maxDistance=2250
 soundClassObject=&Fort_WeaponFoley
 volumeMultiplier=0.226
}
