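# Sound cue for the Smartgun's stacking lock-on auto-aim, played when leaving aim-down-sights.
# A mixer layers three sounds: a quieter heavy pistol ADS-out foley, an auto-aim reticle-close sound, and a mechanical layer.
# The mechanical layer randomly picks one of two equally weighted waves and plays it at a slightly lowered pitch.
# It uses the weapon foley sound class and reload attenuation, with a low overall volume.
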SoundCue /Smartgun/Gameplay/StackingLockons/Sounds/AutoAim_ADS_Out_Cue
root=Mixer_0
nodes [
 {id=Mixer_0, children [Modulator_0,WavePlayer_0,Modulator_2], properties {inputVolume [1.5,1,1.2]}}
 {id=Modulator_0, children [WavePlayer_1], properties {volumeMax=0.7, volumeMin=0.7}}
 {id=Modulator_2, children [Random_0], properties {pitchMax=0.9, pitchMin=0.9}}
 {id=Random_0, children [WavePlayer_2,WavePlayer_3], properties {weights [1,1]}}
 {id=WavePlayer_0, properties {soundWaveAssetPtr=&/AutoAim_Target_Retical_Close_01}}
 {id=WavePlayer_1, properties {soundWaveAssetPtr=/Game/Sounds/Fort_Weapons/ADS/Weapon_PISTOLHEAVY_ADS_OUT_01}}
 {id=WavePlayer_2, properties {soundWaveAssetPtr=&/AutoAim_ADS_Out_Mech_01}}
 {id=WavePlayer_3, properties {soundWaveAssetPtr=&/AutoAim_ADS_Out_Mech_02}}
]
properties {
 attenuationSettings=&Fort_Weapon_Reload_Att
 maxDistance=2250
 soundClassObject=&Fort_WeaponFoley
 volumeMultiplier=0.251
}
