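# Sound cue for the Smartgun auto-aim drop impact, using weapon-drop attenuation and the loot-collect sound class.
# A Random node picks one of three equally weighted impact waves and feeds it through a Modulator for pitch and volume variation.
# The cue slightly raises pitch, lowers volume, and references a shared stop-oldest weapon-drop concurrency set.
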SoundCue /Smartgun/Gameplay/StackingLockons/Sounds/AutoAim_DropImpact_Cue
root=Modulator_0
nodes [
 {id=Modulator_0, children [Random_1]}
 {id=Random_1, children [WavePlayer_4,WavePlayer_5,WavePlayer_6], properties {weights [1,1,1]}}
 {id=WavePlayer_4, properties {soundWaveAssetPtr=&/AutoAim_DropImpact_01}}
 {id=WavePlayer_5, properties {soundWaveAssetPtr=&/AutoAim_DropImpact_02}}
 {id=WavePlayer_6, properties {soundWaveAssetPtr=&/AutoAim_DropImpact_03}}
]
properties {
 attenuationSettings=&Fort_Weapon_Drop_Att
 maxDistance=3000
 pitchMultiplier=1.1
 soundClassObject=&Fort_Loot_Collect
 volumeMultiplier=0.8
 concurrencyOverrides {resolutionRule=PreventNew}
 concurrencySet [&Weapon_Drop_StopOldest]
}
