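# AutoAim_Drop_Cue is the Smartgun auto-aim drop sound cue, stored with the stacking lock-on sounds.
# A Random node picks one of three equally weighted drop waves and feeds it through a Modulator at the root.
# The cue is slightly pitched up and lowered in volume, using weapon-drop attenuation, the loot-collect sound class and the Weapon_Drop_StopOldest concurrency set.
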SoundCue /Smartgun/Gameplay/StackingLockons/Sounds/AutoAim_Drop_Cue
root=Modulator_0
nodes [
 {id=Modulator_0, children [Random_0]}
 {id=Random_0, children [WavePlayer_1,WavePlayer_2,WavePlayer_3], properties {weights [1,1,1]}}
 {id=WavePlayer_1, properties {soundWaveAssetPtr=&/AutoAim_Drop_01}}
 {id=WavePlayer_2, properties {soundWaveAssetPtr=&/AutoAim_Drop_02}}
 {id=WavePlayer_3, properties {soundWaveAssetPtr=&/AutoAim_Drop_03}}
]
properties {
 attenuationSettings=&Fort_Weapon_Drop_Att
 maxDistance=3000
 pitchMultiplier=1.1
 soundClassObject=&Fort_Loot_Collect
 volumeMultiplier=0.6
 concurrencyOverrides {resolutionRule=PreventNew}
 concurrencySet [&Weapon_Drop_StopOldest]
}
