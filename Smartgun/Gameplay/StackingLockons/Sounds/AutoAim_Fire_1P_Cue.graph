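# First-person fire sound cue for the Smartgun auto-aim weapon, using the first-person ranged-weapon sound class and gunfire concurrency.
# A quality-level switch routes its first two tiers to a full mix of randomized close shot, sub layer, delayed pitched-down layer and later, quieter shell casings.
# The last two tiers use a leaner mix with fewer shot and layer variations, no casings, and random nodes excluded from branch culling.
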SoundCue /Smartgun/Gameplay/StackingLockons/Sounds/AutoAim_Fire_1P_Cue
root=QualityLevel_1
nodes [
 {id=Delay_1, children [Modulator_2], properties {delayMax=0.09, delayMin=0.09}}
 {id=Delay_2, children [Random_3], properties {delayMax=0.6, delayMin=0.6}}
 {id=Delay_3, children [Modulator_4], properties {delayMax=0.09, delayMin=0.09}}
 {id=Mixer_2, children [Modulator_0,WavePlayer_8,Delay_1,Modulator_3], properties {inputVolume [1.1,0.7,1.2,0.25]}}
 {id=Mixer_3, children [Modulator_1, WavePlayer_15, Delay_3], properties {inputVolume [1.1,0.6,1.2]}}
 {id=Modulator_0, children [Random_0], properties {pitchMax=1.1, pitchMin=1}}
 {id=Modulator_1, children [Random_2]}
 {id=Modulator_2, children [Random_1], properties {pitchMax=0.9, pitchMin=0.9}}
 {id=Modulator_3, children [Delay_2], properties {pitchMax=0.9, pitchMin=0.8}}
 {id=Modulator_4, children [Random_4], properties {pitchMax=0.9, pitchMin=0.9}}
 {id=QualityLevel_1, children [Mixer_2,Mixer_2,Mixer_3,Mixer_3]}
 {id=Random_0, children [WavePlayer_0,WavePlayer_1,WavePlayer_4,WavePlayer_5,WavePlayer_6,WavePlayer_7], properties {weights [1,1,1,1,1,1]}}
 {id=Random_1, children [WavePlayer_2, WavePlayer_3, WavePlayer_12, WavePlayer_13, WavePlayer_14], properties {weights [1,1,1,1,1]}}
 {id=Random_2, children [WavePlayer_9, WavePlayer_10, WavePlayer_11], properties {bShouldExcludeFromBranchCulling=true, weights [1,1,1]}}
 {
  id=Random_3
  children [WavePlayer_16, WavePlayer_17, WavePlayer_18, WavePlayer_19, WavePlayer_20, WavePlayer_21, WavePlayer_22]
  properties {weights [1,1,1,1,1,1,1]}
 }
 {id=Random_4, children [WavePlayer_23, WavePlayer_24], properties {bShouldExcludeFromBranchCulling=true, weights [1,1]}}
 {id=WavePlayer_0, properties {soundWaveAssetPtr=&/AutoAim_Fire_Close_01}}
 {id=WavePlayer_1, properties {soundWaveAssetPtr=&/AutoAim_Fire_Close_02}}
 {id=WavePlayer_10, properties {soundWaveAssetPtr=&AutoAim_Fire_Close_02}}
 {id=WavePlayer_11, properties {soundWaveAssetPtr=&/AutoAim_Fire_Close_03}}
 {id=WavePlayer_12, properties {soundWaveAssetPtr=&/AutoAim_Fire_Close_Layer_03}}
 {id=WavePlayer_13, properties {soundWaveAssetPtr=&/AutoAim_Fire_Close_Layer_04}}
 {id=WavePlayer_14, properties {soundWaveAssetPtr=&/AutoAim_Fire_Close_Layer_05}}
 {id=WavePlayer_15, properties {soundWaveAssetPtr=&/AutoAim_Fire_Close_Sub_01}}
 {id=WavePlayer_16, properties {soundWaveAssetPtr=&/AutoAim_ShellCasings_2}}
 {id=WavePlayer_17, properties {soundWaveAssetPtr=&/AutoAim_ShellCasings_3}}
 {id=WavePlayer_18, properties {soundWaveAssetPtr=&/AutoAim_ShellCasings_4}}
 {id=WavePlayer_19, properties {soundWaveAssetPtr=&/AutoAim_ShellCasings_5}}
 {id=WavePlayer_2, properties {soundWaveAssetPtr=&/AutoAim_Fire_Close_Layer_01}}
 {id=WavePlayer_20, properties {soundWaveAssetPtr=&/AutoAim_ShellCasings_6}}
 {id=WavePlayer_21, properties {soundWaveAssetPtr=&/AutoAim_ShellCasings_7}}
 {id=WavePlayer_22, properties {soundWaveAssetPtr=&/AutoAim_ShellCasings_8}}
 {id=WavePlayer_23, properties {soundWaveAssetPtr=&AutoAim_Fire_Close_Layer_01}}
 {id=WavePlayer_24, properties {soundWaveAssetPtr=&/AutoAim_Fire_Close_Layer_02}}
 {id=WavePlayer_3, properties {soundWaveAssetPtr=&AutoAim_Fire_Close_Layer_02}}
 {id=WavePlayer_4, properties {soundWaveAssetPtr=&AutoAim_Fire_Close_03}}
 {id=WavePlayer_5, properties {soundWaveAssetPtr=&/AutoAim_Fire_Close_04}}
 {id=WavePlayer_6, properties {soundWaveAssetPtr=&/AutoAim_Fire_Close_05}}
 {id=WavePlayer_7, properties {soundWaveAssetPtr=&/AutoAim_Fire_Close_06}}
 {id=WavePlayer_8, properties {soundWaveAssetPtr=&AutoAim_Fire_Close_Sub_01}}
 {id=WavePlayer_9, properties {soundWaveAssetPtr=&AutoAim_Fire_Close_01}}
]
properties {
 attenuationSettings=&Fort_Weapon_Projectile_FirstPerson_Athena_Att
 maxDistance=2097152
 pitchMultiplier=1.05
 soundClassObject=&Fort_WeaponRanged_Fire_1P_PSM
 volumeMultiplier=0.6
 attenuationOverrides {distanceAlgorithm=NaturalSound, falloffDistance=5000, attenuationShapeExtents {x=1000}}
 concurrencySet [&Weapon_Gunfire_FirstPerson_Conc]
}
