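# Sound cue for the Smartgun auto-aim firing sound, classed as third-person ranged weapon fire and capped by local and global gunfire concurrency sets.
# It mixes close, medium and distant layers, each with its own attenuation, sound class and modulator, plus a sub wave under the close layer.
# Random nodes vary the waves in each layer.
# A quality-level switch uses six variations per layer for the first two levels and three for the last two.
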SoundCue /Smartgun/Gameplay/StackingLockons/Sounds/AutoAim_Fire_Cue
root=QualityLevel_1
nodes [
 {id=Attenuation_0, children [Modulator_0], properties {attenuationSettings=&Fort_Weapon_Projectile_Close_Athena_Att}}
 {
  id=Attenuation_1
  children [Modulator_1]
  properties {attenuationSettings=&Fort_Weapon_Projectile_Medium_Mono_Athena_Att}
 }
 {
  id=Attenuation_5
  children [Modulator_5]
  properties {attenuationSettings=&Fort_Weapon_Projectile_Distant_Mono_Athena_Att}
 }
 {id=Attenuation_6, children [Modulator_6], properties {attenuationSettings=&Fort_Weapon_Projectile_Close_Athena_Att}}
 {
  id=Attenuation_7
  children [Modulator_7]
  properties {attenuationSettings=&Fort_Weapon_Projectile_Medium_Mono_Athena_Att}
 }
 {
  id=Attenuation_8
  children [Modulator_8]
  properties {attenuationSettings=&Fort_Weapon_Projectile_Distant_Mono_Athena_Att}
 }
 {id=Mixer_0, children [SoundClass_0,SoundClass_1,SoundClass_5], properties {inputVolume [1,1,0.7]}}
 {id=Mixer_1, children [Random_3,WavePlayer_6], properties {inputVolume [1,0.45]}}
 {id=Mixer_2, children [SoundClass_6,SoundClass_7,SoundClass_8], properties {inputVolume [1,1,0.7]}}
 {id=Mixer_3, children [Random_9, WavePlayer_34], properties {inputVolume [1,0.45]}}
 {id=Modulator_0, children [Mixer_1]}
 {id=Modulator_1, children [Random_4], properties {pitchMax=1.1, pitchMin=1, volumeMax=1.3, volumeMin=1.2}}
 {id=Modulator_5, children [Random_5]}
 {id=Modulator_6, children [Mixer_3]}
 {id=Modulator_7, children [Random_10], properties {pitchMax=1.1, pitchMin=1, volumeMax=1.3, volumeMin=1.2}}
 {id=Modulator_8, children [Random_11]}
 {id=QualityLevel_1, children [Mixer_0,Mixer_0,Mixer_2,Mixer_2]}
 {id=Random_10, children [WavePlayer_35, WavePlayer_36, WavePlayer_37], properties {bShouldExcludeFromBranchCulling=true, weights [1,1,1]}}
 {id=Random_11, children [WavePlayer_41, WavePlayer_42, WavePlayer_43], properties {bShouldExcludeFromBranchCulling=true, weights [1,1,1]}}
 {id=Random_3, children [WavePlayer_0,WavePlayer_1,WavePlayer_2,WavePlayer_3,WavePlayer_4,WavePlayer_5], properties {weights [1,1,1,1,1,1]}}
 {
  id=Random_4
  children [WavePlayer_7, WavePlayer_8, WavePlayer_9, WavePlayer_10, WavePlayer_11, WavePlayer_12]
  properties {weights [1,1,1,1,1,1]}
 }
 {
  id=Random_5
  children [WavePlayer_13, WavePlayer_14, WavePlayer_15, WavePlayer_16, WavePlayer_17, WavePlayer_18]
  properties {weights [1,1,1,1,1,1]}
 }
 {id=Random_9, children [WavePlayer_19, WavePlayer_20, WavePlayer_30], properties {bShouldExcludeFromBranchCulling=true, weights [1,1,1]}}
 {
  id=SoundClass_0
  children [Attenuation_0]
  properties {soundClassOverride=&Fort_WeaponRanged_Fire_3P_PSM}
 }
 {
  id=SoundClass_1
  children [Attenuation_1]
  properties {soundClassOverride=&Fort_WeaponRanged_Fire_Distant}
 }
 {
  id=SoundClass_5
  children [Attenuation_5]
  properties {soundClassOverride=&Fort_WeaponRanged_Fire_Distant}
 }
 {
  id=SoundClass_6
  children [Attenuation_6]
  properties {soundClassOverride=&Fort_WeaponRanged_Fire_3P_PSM}
 }
 {
  id=SoundClass_7
  children [Attenuation_7]
  properties {soundClassOverride=&Fort_WeaponRanged_Fire_Distant}
 }
 {
  id=SoundClass_8
  children [Attenuation_8]
  properties {soundClassOverride=&Fort_WeaponRanged_Fire_Distant}
 }
 {id=WavePlayer_0, properties {soundWaveAssetPtr=&/AutoAim_Fire_Close_01}}
 {id=WavePlayer_1, properties {soundWaveAssetPtr=&/AutoAim_Fire_Close_02}}
 {id=WavePlayer_10, properties {soundWaveAssetPtr=&/AutoAim_Fire_Medium_04}}
 {id=WavePlayer_11, properties {soundWaveAssetPtr=&/AutoAim_Fire_Medium_05}}
 {id=WavePlayer_12, properties {soundWaveAssetPtr=&/AutoAim_Fire_Medium_06}}
 {id=WavePlayer_13, properties {soundWaveAssetPtr=&/AutoAim_Fire_Distant_01}}
 {id=WavePlayer_14, properties {soundWaveAssetPtr=&/AutoAim_Fire_Distant_02}}
 {id=WavePlayer_15, properties {soundWaveAssetPtr=&/AutoAim_Fire_Distant_03}}
 {id=WavePlayer_16, properties {soundWaveAssetPtr=&/AutoAim_Fire_Distant_04}}
 {id=WavePlayer_17, properties {soundWaveAssetPtr=&/AutoAim_Fire_Distant_05}}
 {id=WavePlayer_18, properties {soundWaveAssetPtr=&/AutoAim_Fire_Distant_06}}
 {id=WavePlayer_19, properties {soundWaveAssetPtr=&AutoAim_Fire_Close_01}}
 {id=WavePlayer_2, properties {soundWaveAssetPtr=&/AutoAim_Fire_Close_03}}
 {id=WavePlayer_20, properties {soundWaveAssetPtr=&AutoAim_Fire_Close_02}}
 {id=WavePlayer_3, properties {soundWaveAssetPtr=&/AutoAim_Fire_Close_04}}
 {id=WavePlayer_30, properties {soundWaveAssetPtr=&AutoAim_Fire_Close_03}}
 {id=WavePlayer_34, properties {soundWaveAssetPtr=&/AutoAim_Fire_Close_Sub_01}}
 {id=WavePlayer_35, properties {soundWaveAssetPtr=&/AutoAim_Fire_Medium_01}}
 {id=WavePlayer_36, properties {soundWaveAssetPtr=&/AutoAim_Fire_Medium_02}}
 {id=WavePlayer_37, properties {soundWaveAssetPtr=&/AutoAim_Fire_Medium_03}}
 {id=WavePlayer_4, properties {soundWaveAssetPtr=&/AutoAim_Fire_Close_05}}
 {id=WavePlayer_41, properties {soundWaveAssetPtr=&AutoAim_Fire_Distant_01}}
 {id=WavePlayer_42, properties {soundWaveAssetPtr=&AutoAim_Fire_Distant_02}}
 {id=WavePlayer_43, properties {soundWaveAssetPtr=&AutoAim_Fire_Distant_03}}
 {id=WavePlayer_5, properties {soundWaveAssetPtr=&/AutoAim_Fire_Close_06}}
 {id=WavePlayer_6, properties {soundWaveAssetPtr=&AutoAim_Fire_Close_Sub_01}}
 {id=WavePlayer_7, properties {soundWaveAssetPtr=&AutoAim_Fire_Medium_01}}
 {id=WavePlayer_8, properties {soundWaveAssetPtr=&AutoAim_Fire_Medium_02}}
 {id=WavePlayer_9, properties {soundWaveAssetPtr=&AutoAim_Fire_Medium_03}}
]
properties {
 bDebug=true
 maxDistance=27500
 pitchMultiplier=1.05
 soundClassObject=&Fort_WeaponRanged_Fire_3P_PSM
 volumeMultiplier=0.6
 concurrencyOverrides {bLimitToOwner=true, maxCount=3, resolutionRule=StopOldest}
 concurrencySet [&Weapon_Gunfire_Local_Conc, &Weapon_Gunfire_Global_Conc]
}
