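# Sound cue for the end of the Smartgun auto-aim reload, part of its stacking lock-on gameplay.
# A random node picks one of two equally weighted wave variations, then a modulator slightly varies pitch and volume before a mixer whose second input is empty.
# It plays as weapon foley with reload attenuation settings, and its concurrency override blocks new instances once the limit is reached.
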SoundCue /Smartgun/Gameplay/StackingLockons/Sounds/AutoAim_ReloadEnd_Cue
root=Mixer_0
nodes [
 {id=Mixer_0, children [Modulator_0,null], properties {inputVolume [1,0.4]}}
 {id=Modulator_0, children [Random_0], properties {pitchMax=1.1, pitchMin=1, volumeMax=0.9, volumeMin=0.8}}
 {id=Random_0, children [WavePlayer_0,WavePlayer_1], properties {weights [1,1]}}
 {id=WavePlayer_0, properties {soundWaveAssetPtr=&/AutoAim_ReloadEnd_01}}
 {id=WavePlayer_1, properties {soundWaveAssetPtr=&/AutoAim_ReloadEnd_02}}
]
properties {
 attenuationSettings=&Fort_Weapon_Reload_Att
 bOverrideConcurrency=true
 maxDistance=2250
 soundClassObject=&Fort_WeaponFoley
 volumeMultiplier=1.002
 attenuationOverrides {falloffDistance=4100, attenuationShapeExtents {x=900}}
 concurrencyOverrides {resolutionRule=PreventNew}
}
