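# Sound cue for the reload insert sound of the Smartgun's auto-aim, stacking lock-on weapon.
# It randomly picks one of three equally weighted insert samples and feeds it through a modulator that slightly raises pitch at random.
# It plays as weapon foley with shared reload attenuation.
# Its concurrency override prevents new instances once the limit is reached.
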SoundCue /Smartgun/Gameplay/StackingLockons/Sounds/AutoAim_ReloadInsert_Cue
root=Modulator_1
nodes [
 {id=Modulator_1, children [Random_0], properties {pitchMax=1.1, pitchMin=1}}
 {id=Random_0, children [WavePlayer_0,WavePlayer_1,WavePlayer_2], properties {weights [1,1,1]}}
 {id=WavePlayer_0, properties {soundWaveAssetPtr=&/AutoAim_ReloadInsert_01}}
 {id=WavePlayer_1, properties {soundWaveAssetPtr=&/AutoAim_ReloadInsert_02}}
 {id=WavePlayer_2, properties {soundWaveAssetPtr=&/AutoAim_ReloadInsert_03}}
]
properties {
 attenuationSettings=&Fort_Weapon_Reload_Att
 bOverrideConcurrency=true
 maxDistance=2250
 soundClassObject=&Fort_WeaponFoley
 volumeMultiplier=0.652
 attenuationOverrides {falloffDistance=4100, attenuationShapeExtents {x=900}}
 concurrencyOverrides {resolutionRule=PreventNew}
}
