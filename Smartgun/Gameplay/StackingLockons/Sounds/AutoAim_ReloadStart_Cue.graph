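# Sound cue for the start of a Smartgun auto-aim reload, by its name and path, played as weapon foley.
# A mixer layers a randomly chosen reload-start sound, passed through a modulator, with a random magazine impact offset by a short fixed delay.
# Each layer picks from three equally weighted waves.
# It uses weapon reload attenuation, and its concurrency override blocks new instances once the limit is reached.
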SoundCue /Smartgun/Gameplay/StackingLockons/Sounds/AutoAim_ReloadStart_Cue
root=Mixer_0
nodes [
 {id=Delay_0, children [Random_1], properties {delayMax=0.15, delayMin=0.15}}
 {id=Mixer_0, children [Modulator_0,Delay_0], properties {inputVolume [1,1]}}
 {id=Modulator_0, children [Random_0]}
 {id=Random_0, children [WavePlayer_0,WavePlayer_1,WavePlayer_2], properties {weights [1,1,1]}}
 {id=Random_1, children [WavePlayer_3,WavePlayer_4,WavePlayer_5], properties {weights [1,1,1]}}
 {id=WavePlayer_0, properties {soundWaveAssetPtr=&/AutoAim_ReloadStart_01}}
 {id=WavePlayer_1, properties {soundWaveAssetPtr=&/AutoAim_ReloadStart_02}}
 {id=WavePlayer_2, properties {soundWaveAssetPtr=&/AutoAim_ReloadStart_03}}
 {id=WavePlayer_3, properties {soundWaveAssetPtr=&/AutoAim_Magazine_Impact_2}}
 {id=WavePlayer_4, properties {soundWaveAssetPtr=&/AutoAim_Magazine_Impact_3}}
 {id=WavePlayer_5, properties {soundWaveAssetPtr=&/AutoAim_Magazine_Impact_4}}
]
properties {
 attenuationSettings=&Fort_Weapon_Reload_Att
 bOverrideConcurrency=true
 maxDistance=2250
 soundClassObject=&Fort_WeaponFoley
 volumeMultiplier=0.677
 attenuationOverrides {falloffDistance=4100, attenuationShapeExtents {x=900}}
 concurrencyOverrides {resolutionRule=PreventNew}
}
