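# Sound cue for the Smartgun's stacking lock-ons, likely played as the auto-aim target count changes.
# A single wave player feeds the output directly at reduced volume, with no randomization or modulation.
# It uses the weapon foley sound class and weapon reload attenuation with its own falloff and shape overrides.
# Its own concurrency override prevents new instances from starting once the limit is reached.
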SoundCue /Smartgun/Gameplay/StackingLockons/Sounds/AutoAim_Target_Count_Cue
root=WavePlayer_0
nodes [{id=WavePlayer_0, properties {soundWaveAssetPtr=&/AutoAim_Target_Count_A_01}}]
properties {
 attenuationSettings=&Fort_Weapon_Reload_Att
 bOverrideConcurrency=true
 maxDistance=2250
 soundClassObject=&Fort_WeaponFoley
 volumeMultiplier=0.476
 attenuationOverrides {falloffDistance=4100, attenuationShapeExtents {x=900}}
 concurrencyOverrides {resolutionRule=PreventNew}
}
