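# Sound cue for the Smartgun's stacking lock-on feature, named as the sound for an auto-aim target lock appearing.
# A single wave plays through a modulator that applies a fixed volume boost without randomization, with the cue pitched slightly down.
# It uses the gadgets and items sound class and borrows the security turret's attenuation.
# A concurrency override allows one instance per owner, stopping the oldest when a new one starts.
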SoundCue /Smartgun/Gameplay/StackingLockons/Sounds/AutoAim_Target_Lock_Appears_Cue
root=Modulator_1
nodes [
 {id=Modulator_1, children [WavePlayer_1], properties {pitchMax=1, pitchMin=1, volumeMax=1.4, volumeMin=1.4}}
 {id=WavePlayer_1, properties {soundWaveAssetPtr=&/AutoAim_Target_Lock_Appears_01}}
]
properties {
 attenuationSettings=/Game/Athena/Sounds/Gadgets/SecurityTurret/Security_Turret_Sphere_Att
 bOverrideConcurrency=true
 maxDistance=3500
 pitchMultiplier=0.9
 soundClassObject=&Fort_Gadgets_Items
 concurrencyOverrides {bLimitToOwner=true, maxCount=1, resolutionRule=StopOldest}
}
