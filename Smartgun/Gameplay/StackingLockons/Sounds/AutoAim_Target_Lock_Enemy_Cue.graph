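# Sound cue for the Smartgun's stacking lock-on feature, played when auto-aim locks onto an enemy target.
# A mixer layers a one-shot lock sound, pitched slightly down, with a louder-weighted looping sound.
# An enveloper keeps the loop silent for about half a second, then fades it in to full by about one second.
# It uses the security turret's attenuation and overrides concurrency to three voices, stopping the quietest.
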SoundCue /Smartgun/Gameplay/StackingLockons/Sounds/AutoAim_Target_Lock_Enemy_Cue
root=Mixer_0
nodes [
 {
  id=Enveloper_0
  children [WavePlayer_0]
  properties {
   pitchCurve {editorCurveData {keys [{time=0}]}}
   volumeCurve {editorCurveData {keys [<at 0.0059 0.0058>,<at 0.4911 -0.0039>,<at 0.5773 0.2023>,<at 0.7297 0.4556>,<at 0.8948 0.7266>,<at 1.027 0.9858>]}}
  }
 }
 {id=Mixer_0, children [Modulator_0,Enveloper_0], properties {inputVolume [1,2]}}
 {id=Modulator_0, children [WavePlayer_7], properties {pitchMax=0.9, pitchMin=0.9}}
 {id=WavePlayer_0, properties {bLooping=true, soundWaveAssetPtr=&/AutoAim_Target_Lock_Enemy_Loop_01}}
 {id=WavePlayer_7, properties {soundWaveAssetPtr=&/AutoAim_Target_Lock_Enemy_V2_02}}
]
properties {
 attenuationSettings=/Game/Athena/Sounds/Gadgets/SecurityTurret/Security_Turret_Sphere_Att
 bOverrideConcurrency=true
 maxDistance=3500
 soundClassObject=&Fort_Gadgets_Items
 volumeMultiplier=0.5
 attenuationOverrides {
  bAttenuate=false
  bEnableFocusInterpolation=true
  bEnableListenerFocus=true
  falloffDistance=1200
  nonFocusVolumeAttenuation=0.5
  stereoSpread=10
  attenuationShapeExtents {x=0}
 }
 concurrencyOverrides {maxCount=3, resolutionRule=StopQuietest}
}
