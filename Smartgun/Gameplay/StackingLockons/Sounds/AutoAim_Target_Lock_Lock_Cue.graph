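# Sound cue for the lock sound of the Smartgun's stacking auto-aim target lock-on.
# A mixer plays three lock waves layered together, with the third mixed in much quieter.
# It plays in the gadgets and items sound class with the security turret's sphere attenuation asset assigned.
# Concurrency is overridden to a small instance cap that stops the quietest voice.
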SoundCue /Smartgun/Gameplay/StackingLockons/Sounds/AutoAim_Target_Lock_Lock_Cue
root=Mixer_0
nodes [
 {id=Mixer_0, children [WavePlayer_0,WavePlayer_1,WavePlayer_2], properties {inputVolume [1,1,0.3]}}
 {id=WavePlayer_0, properties {soundWaveAssetPtr=&/AutoAim_Target_Lock_Lock_01}}
 {id=WavePlayer_1, properties {soundWaveAssetPtr=&/AutoAim_Target_Lock_Lock_02}}
 {id=WavePlayer_2, properties {soundWaveAssetPtr=&/AutoAim_Target_Lock_Lock_03}}
]
properties {
 attenuationSettings=/Game/Athena/Sounds/Gadgets/SecurityTurret/Security_Turret_Sphere_Att
 bOverrideConcurrency=true
 maxDistance=3500
 soundClassObject=&Fort_Gadgets_Items
 volumeMultiplier=0.9
 attenuationOverrides {
  bEnableFocusInterpolation=true
  bEnableListenerFocus=true
  falloffDistance=1200
  nonFocusVolumeAttenuation=0.5
  stereoSpread=500
  attenuationShapeExtents {x=0}
 }
 concurrencyOverrides {maxCount=3, resolutionRule=StopQuietest}
}
