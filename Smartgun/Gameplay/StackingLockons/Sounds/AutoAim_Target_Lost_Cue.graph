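# AutoAim_Target_Lost_Cue is a SoundCue for the Smartgun's stacking lock-on system, named for when auto-aim loses its target.
# A mixer layers the Target Lost sound with a louder, slightly delayed mechanical layer.
# Overridden concurrency allows one instance per owner, stopping the oldest.
# It plays in the gadget items sound class at reduced volume, borrowing the security turret's spherical attenuation.
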SoundCue /Smartgun/Gameplay/StackingLockons/Sounds/AutoAim_Target_Lost_Cue
root=Mixer_0
nodes [
 {id=Delay_0, children [WavePlayer_1], properties {delayMax=0.1, delayMin=0.1}}
 {id=Mixer_0, children [WavePlayer_0,Delay_0], properties {inputVolume [1,2]}}
 {id=WavePlayer_0, properties {soundWaveAssetPtr=&/AutoAim_Target_Lost_01}}
 {id=WavePlayer_1, properties {soundWaveAssetPtr=&/AutoAim_Target_Lost_Mech_01}}
]
properties {
 attenuationSettings=/Game/Athena/Sounds/Gadgets/SecurityTurret/Security_Turret_Sphere_Att
 bOverrideConcurrency=true
 maxDistance=3500
 soundClassObject=&Fort_Gadgets_Items
 volumeMultiplier=0.4
 concurrencyOverrides {bLimitToOwner=true, maxCount=1, resolutionRule=StopOldest}
}
